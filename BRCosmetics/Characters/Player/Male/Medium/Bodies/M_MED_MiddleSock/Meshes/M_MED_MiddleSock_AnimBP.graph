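# Animation Blueprint for the M_MED_MiddleSock player body that corrects its left and right shoulder pad bones.
# It copies the parent mesh pose into a Control Rig fed by weapon and crouch flags, then constrains pad bones to the upper arms.
# Additive bone offsets follow, weighted by arm-to-clavicle/spine rotation deltas and targeting, skydiving, emote, crouch and melee states.
# The event graph computes those weights each update; Frontend correction is always false.
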AnimBlueprint /BRCosmetics/Characters/Player/Male/Medium/Bodies/M_MED_MiddleSock/Meshes/M_MED_MiddleSock_AnimBP
target_skeleton=&/M_MED_MiddleSock_Skeleton
graphs [
 AnimGraph {nodes [
   n1 = "Transform (Modify) Bone - Bone: dyn_shoulderpad_l"!(
    ComponentPose=Local To Component(LocalPose=Use cached pose 'CONSTRAINT'())
    bAlphaBoolEnabled:bool=True
    Translation="0.474,0.003,0.250"
    Rotation="1.776,6.289,-3.573"
    Scale="1,1,1"
   ) -> (Pose) {
    alphaInputType=Bool
    rotationMode=BMM_Additive
    rotationSpace=BCS_BoneSpace
    translationMode=BMM_Additive
    translationSpace=BCS_BoneSpace
    alphaBoolBlend {blendOption=Linear}
    boneToModify {boneName=dyn_shoulderpad_l}
    rotation=<rot 1.776 6.289 -3.573>
    translation=<0.474 0.003 0.25>
   }
   n2 = "Transform (Modify) Bone - Bone: dyn_shoulderpad_r"!(ComponentPose=@n1, bAlphaBoolEnabled:bool=True, Translation="-0.6,0,-0.5", Rotation="-3.498,4.335,0.195", Scale="1,1,1") -> (Pose) {
    alphaInputType=Bool
    rotationMode=BMM_Additive
    rotationSpace=BCS_BoneSpace
    translationMode=BMM_Additive
    translationSpace=BCS_BoneSpace
    alphaBoolBlend {blendOption=Linear}
    boneToModify {boneName=dyn_shoulderpad_r}
    rotation=<rot -3.498 4.335 0.195>
    translation {x=-0.6, z=-0.5}
   }
   n3 = "Transform (Modify) Bone - Bone: dyn_shoulderpad_r"!(
    ComponentPose=@n2
    Alpha:float=Get right shoulder pad arm twist corr() -> (right shoulder pad arm twist corr:double)
    Translation="2.264,0.209,0.133"
    Rotation="20.877,1.577,1.155"
    Scale="1,1,1"
   ) -> (Pose) {
    rotationMode=BMM_Additive
    rotationSpace=BCS_BoneSpace
    translationMode=BMM_Additive
    translationSpace=BCS_BoneSpace
    alphaBoolBlend {blendOption=Linear}
    alphaScaleBiasClamp {bInterpResult=true}
    boneToModify {boneName=dyn_shoulderpad_r}
    rotation=<rot 20.877 1.577 1.155>
    translation=<2.264 0.209 0.133>
   }
   n4 = "Transform (Modify) Bone - Bone: dyn_shoulderpad_l"!(
    ComponentPose=@n3
    Alpha:float=Get left shoulder pad arm twist corr() -> (left shoulder pad arm twist corr:double)
    Translation="-0.589,-0.234,-0.229"
    Rotation="10.842,6.762,-2.982"
    Scale="1,1,1"
   ) -> (Pose) {
    rotationMode=BMM_Additive
    rotationSpace=BCS_BoneSpace
    translationMode=BMM_Additive
    translationSpace=BCS_BoneSpace
    alphaBoolBlend {blendOption=Linear}
    alphaScaleBiasClamp {bInterpResult=true}
    boneToModify {boneName=dyn_shoulderpad_l}
    rotation=<rot 10.842 6.762 -2.982>
    translation=<-0.589 -0.234 -0.229>
   }
   n5 = "Transform (Modify) Bone - Bone: dyn_shoulderpad_r"(
    ComponentPose=@n4
    Alpha=Get right shoulder pad up corr() -> (right shoulder pad up corr:double)
    Translation="-1.348,-5.594,-0.433"
    Rotation="-8.351,-6.654,11.424"
    Scale="1,1,1"
   ) {
    rotationMode=BMM_Additive
    rotationSpace=BCS_BoneSpace
    translationMode=BMM_Additive
    translationSpace=BCS_BoneSpace
    alphaBoolBlend {blendOption=Linear}
    boneToModify {boneName=dyn_shoulderpad_r}
    rotation=<rot -8.351 -6.654 11.424>
    translation=<-1.348 -5.594 -0.433>
   }
   n6 = "Transform (Modify) Bone - Bone: dyn_shoulderpad_l"(
    ComponentPose=@n5
    Alpha=Get left shoulder pad up corr() -> (left shoulder pad up corr:double)
    Translation="-1.221,5.031,0.037"
    Rotation="8.351,6.654,11.424"
    Scale="1,1,1"
   ) {
    rotationMode=BMM_Additive
    rotationSpace=BCS_BoneSpace
    translationMode=BMM_Additive
    translationSpace=BCS_BoneSpace
    alphaBoolBlend {blendOption=Linear}
    boneToModify {boneName=dyn_shoulderpad_l}
    rotation=<rot 8.351 6.654 11.424>
    translation=<-1.221 5.031 0.037>
   }
   n7 = Get Frontend corr() -> (Frontend corr:bool)
   n8 = "Transform (Modify) Bone - Bone: dyn_shoulderpad_r"!(ComponentPose=@n6, bAlphaBoolEnabled:bool=@n7, Translation="-1.163,0.134,-0.161", Rotation="-14.058,-2.562,-2.604", Scale="1,1,1") -> (Pose) {
    alphaInputType=Bool
    rotationMode=BMM_Additive
    rotationSpace=BCS_BoneSpace
    translationMode=BMM_Additive
    translationSpace=BCS_BoneSpace
    alphaBoolBlend {blendInTime=1, blendOption=Linear, blendOutTime=0.3}
    boneToModify {boneName=dyn_shoulderpad_r}
    rotation=<rot -14.058 -2.5619 -2.6041>
    translation=<-1.163 0.134 -0.161>
   }
   n9 = "Transform (Modify) Bone - Bone: dyn_shoulderpad_l"!(ComponentPose=@n8, bAlphaBoolEnabled:bool=@n7, Translation="0.135,0.083,-0.217", Rotation="-8.782,-8.121,1.216", Scale="1,1,1") -> (Pose) {
    alphaInputType=Bool
    rotationMode=BMM_Additive
    rotationSpace=BCS_BoneSpace
    translationMode=BMM_Additive
    translationSpace=BCS_BoneSpace
    alphaBoolBlend {blendInTime=0.8, blendOption=Linear, blendOutTime=0.3}
    boneToModify {boneName=dyn_shoulderpad_l}
    rotation=<rot -8.782 -8.121 1.216>
    translation=<0.135 0.083 -0.217>
   }
   n10 = Get Melee corr idle() -> (Melee corr idle:bool)
   n11 = "Transform (Modify) Bone - Bone: dyn_shoulderpad_l"(ComponentPose=@n9, bAlphaBoolEnabled=@n10, Translation="0.874,0.193,1.590", Rotation="5.646,9.212,17.283", Scale="1,1,1") {
    alphaInputType=Bool
    rotationMode=BMM_Additive
    rotationSpace=BCS_BoneSpace
    translationMode=BMM_Additive
    translationSpace=BCS_BoneSpace
    alphaBoolBlend {blendOption=Linear}
    boneToModify {boneName=dyn_shoulderpad_l}
    rotation=<rot 5.646 9.212 17.283>
    translation=<0.874 0.193 1.59>
   }
   n12 = "Transform (Modify) Bone - Bone: dyn_shoulderpad_r"(ComponentPose=@n11, bAlphaBoolEnabled=@n10, Translation="-0.500,0.423,1.241", Rotation="4.556,7.675,-15.225", Scale="1,1,1") {
    alphaInputType=Bool
    rotationMode=BMM_Additive
    rotationSpace=BCS_BoneSpace
    translationMode=BMM_Additive
    translationSpace=BCS_BoneSpace
    alphaBoolBlend {blendOption=Linear}
    boneToModify {boneName=dyn_shoulderpad_r}
    rotation=<rot 4.556 7.675 -15.225>
    translation=<-0.5 0.423 1.241>
   }
   n13 = "Transform (Modify) Bone - Bone: dyn_shoulderpad_l"!(
    ComponentPose=@n12
    Alpha:float=Get Left Pad crouch melee attacking alpha() -> (Left Pad crouch melee attacking alpha:double)
    Translation="-3.432,0.568,1.450"
    Rotation="9.996,2.448,20"
    Scale="1,1,1"
   ) -> (Pose) {
    rotationMode=BMM_Additive
    rotationSpace=BCS_BoneSpace
    translationMode=BMM_Additive
    translationSpace=BCS_BoneSpace
    alphaBoolBlend {blendOption=Linear}
    boneToModify {boneName=dyn_shoulderpad_l}
    rotation=<rot 9.996 2.448 20>
    translation=<-3.432 0.568 1.45>
   }
   n14 = "Transform (Modify) Bone - Bone: dyn_shoulderpad_l"(
    ComponentPose=@n13
    Alpha=Get Left Pad crouch targeting up alpha() -> (Left Pad crouch targeting up alpha:float)
    Translation="-0.308,9.596,0.663"
    Rotation="6.276,-13.853,28.607"
    Scale="1,1,1"
   ) {
    rotationMode=BMM_Additive
    rotationSpace=BCS_BoneSpace
    translationMode=BMM_Additive
    translationSpace=BCS_BoneSpace
    alphaBoolBlend {blendOption=Linear}
    boneToModify {boneName=dyn_shoulderpad_l}
    rotation=<rot 6.276 -13.853 28.607>
    translation=<-0.308 9.596 0.663>
   }
   Output Pose(Result=Component To Local(ComponentPose=@n14)) {layerGroup=None}
   n15 = Control Rig!(
    Source=Copy Pose From Mesh() {bUseAttachedParent=true}
    Alpha
    Rocket launcher corr:bool=Get Rocket launcher corr() -> (Rocket launcher corr:bool)
    Melee corr:bool=Get Melee corr() -> (Melee corr:bool)
    Minigun corr:bool=Get Minigun corr() -> (Minigun corr:bool)
    Crouch:bool=Get bIsCrouching()
   ) -> (Pose) {
    alphaBoolBlend {blendOption=Linear}
    controlRigAssetReference {blueprintRigClass=&/M_MED_MiddleSock_CtrlRig.M_MED_MiddleSock_CtrlRig_C}
   }
   Save cached pose 'CONTROL RIG'(Pose=@n15) {cacheName=CONTROL RIG}
   n16 = "Constraint - dyn_shoulderpad_base_r (PPPPT)"!(
    ComponentPose=Local To Component(LocalPose=Use cached pose 'CONTROL RIG'())
    Alpha=1
    ConstraintWeights_0=0
    ConstraintWeights_1:float=0
    ConstraintWeights_2:float=1
    ConstraintWeights_3:float=0
    ConstraintWeights_4:float=0
   ) -> (Pose) {
    alphaBoolBlend {blendOption=Linear}
    boneToModify {boneName=dyn_shoulderpad_base_r}
    constraintSetup [
     {targetBone {boneName=clavicle_r}}
     {targetBone {boneName=spine_04}}
     {targetBone {boneName=upperarm_twist_01_r}}
     {targetBone {boneName=upperarm_r}}
     {transformType=Translation, targetBone {boneName=None}}
    ]
    constraintWeights [1,1,1,1,1]
   }
   n17 = "Constraint - dyn_shoulderpad_r_bk (PPPPT)"!(
    ComponentPose=@n16
    Alpha=0.3
    ConstraintWeights_0=0
    ConstraintWeights_1:float=0
    ConstraintWeights_2:float=0.4
    ConstraintWeights_3:float=0.6
    ConstraintWeights_4:float=0
   ) -> (Pose) {
    alphaBoolBlend {blendOption=Linear}
    boneToModify {boneName=dyn_shoulderpad_r_bk}
    constraintSetup [
     {targetBone {boneName=clavicle_r}}
     {targetBone {boneName=spine_04}}
     {targetBone {boneName=upperarm_twist_01_r}}
     {targetBone {boneName=upperarm_r}}
     {transformType=Translation, targetBone {boneName=None}}
    ]
    constraintWeights [1,1,1,1,1]
   }
   n18 = "Constraint - dyn_shoulderpad_r_fr (PPPPT)"!(
    ComponentPose=@n17
    Alpha=0.3
    ConstraintWeights_0=0
    ConstraintWeights_1:float=0
    ConstraintWeights_2:float=0.4
    ConstraintWeights_3:float=0.6
    ConstraintWeights_4:float=0
   ) -> (Pose) {
    alphaBoolBlend {blendOption=Linear}
    boneToModify {boneName=dyn_shoulderpad_r_fr}
    constraintSetup [
     {targetBone {boneName=clavicle_r}}
     {targetBone {boneName=spine_04}}
     {targetBone {boneName=upperarm_twist_01_r}}
     {targetBone {boneName=upperarm_r}}
     {transformType=Translation, targetBone {boneName=None}}
    ]
    constraintWeights [1,1,1,1,1]
   }
   n19 = "Constraint - dyn_shoulderpad_base_l (PPPPT)"!(
    ComponentPose=@n18
    Alpha=1
    ConstraintWeights_0=0
    ConstraintWeights_1:float=0
    ConstraintWeights_2:float=1
    ConstraintWeights_3:float=0
    ConstraintWeights_4:float=0
   ) -> (Pose) {
    alphaBoolBlend {blendOption=Linear}
    boneToModify {boneName=dyn_shoulderpad_base_l}
    constraintSetup [
     {targetBone {boneName=clavicle_r}}
     {targetBone {boneName=spine_04}}
     {targetBone {boneName=upperarm_twist_01_l}}
     {targetBone {boneName=upperarm_l}}
     {transformType=Translation, targetBone {boneName=None}}
    ]
    constraintWeights [1,1,1,1,1]
   }
   n20 = "Constraint - dyn_shoulderpad_base_l_bk (PPPPT)"!(
    ComponentPose=@n19
    Alpha=0.3
    ConstraintWeights_0=0
    ConstraintWeights_1:float=0
    ConstraintWeights_2:float=0.4
    ConstraintWeights_3:float=0.6
    ConstraintWeights_4:float=0
   ) -> (Pose) {
    alphaBoolBlend {blendOption=Linear}
    boneToModify {boneName=dyn_shoulderpad_base_l_bk}
    constraintSetup [
     {targetBone {boneName=clavicle_l}}
     {targetBone {boneName=spine_04}}
     {targetBone {boneName=upperarm_twist_01_l}}
     {targetBone {boneName=upperarm_l}}
     {transformType=Translation, targetBone {boneName=None}}
    ]
    constraintWeights [1,1,1,1,1]
   }
   n21 = "Constraint - dyn_shoulderpad_l_fr (PPPPT)"!(
    ComponentPose=@n20
    Alpha=0.3
    ConstraintWeights_0=0
    ConstraintWeights_1:float=0
    ConstraintWeights_2:float=0.4
    ConstraintWeights_3:float=0.6
    ConstraintWeights_4:float=0
   ) -> (Pose) {
    alphaBoolBlend {blendOption=Linear}
    boneToModify {boneName=dyn_shoulderpad_l_fr}
    constraintSetup [
     {targetBone {boneName=clavicle_l}}
     {targetBone {boneName=spine_04}}
     {targetBone {boneName=upperarm_twist_01_l}}
     {targetBone {boneName=upperarm_l}}
     {transformType=Translation, targetBone {boneName=None}}
    ]
    constraintWeights [1,1,1,1,1]
   }
   n22 = "Constraint - dyn_shoulderpad_base_r (PP)"!(ComponentPose=@n21, Alpha=1, ConstraintWeights_0=0, ConstraintWeights_1:float=1) -> (Pose) {
    alphaBoolBlend {blendOption=Linear}
    boneToModify {boneName=dyn_shoulderpad_base_r}
    constraintSetup [{targetBone {boneName=clavicle_r}}, {targetBone {boneName=upperarm_twist_01_r}}]
    constraintWeights [1,1]
   }
   n23 = "Constraint - dyn_shoulderpad_base_l (PP)"!(ComponentPose=@n22, Alpha=1, ConstraintWeights_0=0, ConstraintWeights_1:float=1) -> (Pose) {
    alphaBoolBlend {blendOption=Linear}
    boneToModify {boneName=dyn_shoulderpad_base_l}
    constraintSetup [{targetBone {boneName=clavicle_l}}, {targetBone {boneName=upperarm_twist_01_l}}]
    constraintWeights [1,1]
   }
   Save cached pose 'CONSTRAINT'(Pose=Component To Local(ComponentPose=@n23)) {cacheName=CONSTRAINT}
  ]}
 EventGraph {nodes [
   Event BlueprintUpdateAnimation()
   > n1 = Sequence!(execute) -> (then_0, then_1, then_2:Exec)
   n2 = Cast To FortPlayerAnimInstance(execute=@n1.then_1, Object=GetAnimInstance(self=GetOwnerSkeletalMeshComponent()))
   > Set Assault Rifle(Assault Rifle:bool=false) -> (Output_Get:bool)
   > Set Rocket launcher corr(Rocket launcher corr:bool=false) -> (Output_Get:bool)
   > Set Melee corr(Melee corr:bool=false) -> (Output_Get:bool)
   > Set Minigun corr(Minigun corr:bool=false) -> (Output_Get:bool)
   > n3 = Switch on EFortWeaponCoreAnimation!(execute, Selection:EFortWeaponCoreAnimation Enum=Get WeaponCoreAnim(self=@n2)) -> (
    Melee:Exec
    Pistol:Exec
    Shotgun:Exec
    PaperBlueprint:Exec
    Rifle:Exec
    MeleeOneHand:Exec
    MachinePistol:Exec
    RocketLauncher:Exec
    GrenadeLauncher:Exec
    GoingCommando:Exec
    AssaultRifle:Exec
    TacticalShotgun:Exec
    SniperRifle:Exec
    TrapPlacement:Exec
    ShoulderLauncher:Exec
    AbilityDecoTool:Exec
    Crossbow:Exec
    C4:Exec
    RemoteControl:Exec
    DualWield:Exec
    AR_BullPup:Exec
    AR_ForwardGrip:Exec
    MedPackPaddles:Exec
    SMG_P90:Exec
    AR_DrumGun:Exec
    Consumable_Small:Exec
    Consumable_Large:Exec
    Balloon:Exec
    MountedTurret:Exec
    CreativeTool:Exec
    ExplosiveBow:Exec
    AshtonIndigo:Exec
    AshtonChicago:Exec
    MeleeDualWield:Exec
    Unarmed:Exec
   )
   Set Assault Rifle(execute=@n3.AssaultRifle, Assault Rifle=true)
   Set Minigun corr(execute=@n3.GoingCommando, Minigun corr=true)
   Set Melee corr(execute=[@n3.Melee, @n3.MeleeDualWield, @n3.MeleeOneHand], Melee corr=true)
   Set Rocket launcher corr(execute=@n3.RocketLauncher, Rocket launcher corr=true)
   n4 = Event OnCharacterCustomizationLoaded()
   > n5 = Set Skeletal Mesh(Skeletal Mesh:Skeletal Mesh Component Object Reference=@n4.CustomBaseMesh) -> (Output_Get:Skeletal Mesh Component Object Reference)
   Set left shoulder pad arm twist corr(
    execute=@n1.then_0
    left shoulder pad arm twist corr:double=GetAlphaFromReferenceTransformDelta(SkeletalMeshComponent=@n5, BoneName=upperarm_l, BaseName=spine_05, DeltaComponentType=EulerX, InputRange=(X=20,Y=3), TwistAxis="1,0,0")
   ) -> (Output_Get:double)
   > Set right shoulder pad arm twist corr(right shoulder pad arm twist corr:double=SelectFloat(
     A=0
     B=GetAlphaFromReferenceTransformDelta(SkeletalMeshComponent=@n5, BoneName=upperarm_r, BaseName=spine_05, DeltaComponentType=EulerX, InputRange=(X=40,Y=3), TwistAxis="1,0,0")
     bPickA=Get bIsTargeting()
    )) -> (Output_Get:double)
   n6 = GetAlphaFromReferenceTransformDelta(SkeletalMeshComponent=@n5, BoneName=upperarm_r, BaseName=clavicle_r, DeltaComponentType=EulerY, InputRange=(X=25,Y=60), TwistAxis="1,0,0")
   > Set right shoulder pad up corr(right shoulder pad up corr:double=SelectFloat(A=float * float(A=@n6, B=0.6) {is=title K2Node_PromotableOperator}, B=@n6, bPickA=Get bIsSkyDiving())) -> (Output_Get:double)
   n7 = AND Boolean(A=NOT Boolean(A=Get bIsPlayingEmote()), B=NOT Boolean(A=Get bIsCrouching()))
   n8 = GetAlphaFromReferenceTransformDelta(SkeletalMeshComponent=@n5, BoneName=upperarm_l, BaseName=clavicle_l, DeltaComponentType=EulerY, InputRange=(X=40,Y=60), TwistAxis="1,0,0")
   n9 = SelectFloat(A=float * float(A=@n8, B=0.1), B=@n8, bPickA=Get bIsSkyDiving())
   n10 = float * float(A=@n9, B=0.8)
   > Set left shoulder pad up corr(left shoulder pad up corr:double=FInterpTo(Current=SelectFloat(A=@n10, B=@n9, bPickA=@n7), Target=0, DeltaTime=0.1, InterpSpeed=0.1)) -> (Output_Get:double)
   > Set Frontend corr(Frontend corr:bool=false) -> (Output_Get:bool)
   n11 = SelectFloat(
    A=GetAlphaFromReferenceTransformDelta(SkeletalMeshComponent=@n5, BoneName=upperarm_l, BaseName=clavicle_l, InputRange=(X=55,Y=90), TwistAxis="1,0,0")
    B=0
    bPickA="float > float"(A=GetAlphaFromReferenceTransformDelta(SkeletalMeshComponent=@n5, BoneName=upperarm_l, BaseName=clavicle_l, DeltaComponentType=EulerY, InputRange=(X=-20,Y=15), TwistAxis="1,0,0"))
   )
   n12 = AND Boolean(A=Get bIsPlayingMeleeAnim(), B=Get bIsCrouching())
   > Set Left Pad crouch melee attacking alpha(Left Pad crouch melee attacking alpha:double=FInterpTo(Current=float * float(A=@n11, B:bool=@n12), Target=1, DeltaTime=0.2, InterpSpeed=0.2)) -> (Output_Get:double)
   n13 = GetAlphaFromReferenceTransformDelta(SkeletalMeshComponent=@n5, BoneName=upperarm_l, BaseName=clavicle_l, DeltaComponentType=EulerY, InputRange=(X=65,Y=45), TwistAxis="0,0,1")
   n14 = SelectFloat(
    A=float + float(
     A=GetAlphaFromReferenceTransformDelta(SkeletalMeshComponent=@n5, BoneName=upperarm_l, BaseName=clavicle_l, InputRange=(X=-110,Y=-170), TwistAxis="0,0,1")
     B=GetAlphaFromReferenceTransformDelta(SkeletalMeshComponent=@n5, BoneName=upperarm_l, BaseName=clavicle_l, InputRange=(X=110,Y=150), TwistAxis="0,0,1")
    )
    B=0
    bPickA="float > float"(A=@n13, B=0.5)
   )
   n15 = OR Boolean(
    A="Equal (Float)"(A=@n13, B=1)
    B="float > float"(
     A=GetAlphaFromReferenceTransformDelta(SkeletalMeshComponent=@n5, BoneName=upperarm_l, BaseName=clavicle_l, DeltaComponentType=EulerX, InputRange=(X=0,Y=73), TwistAxis="0,0,1")
     B=0.5
    )
   )
   > Set Left Pad crouch targeting up alpha(Left Pad crouch targeting up alpha:float=FInterpTo(Current=SelectFloat(A=@n14, B=0, bPickA=@n15), Target=Get left shoulder pad up corr(), DeltaTime=0.1, InterpSpeed=0.1)) -> (Output_Get:float)
   n16 = Get bIsTargeting()
   n17 = Get bIsPlayingMeleeAnim()
   > Set Melee corr idle(Melee corr idle:bool=AND Boolean!(A=Get Melee corr(), B=NOT Boolean(A=@n17), C:bool=NOT Boolean(A=@n16), D:bool=Get bIsCrouching()) -> (ReturnValue)) -> (Output_Get:bool)
   GetAlphaFromReferenceTransformDelta(SkeletalMeshComponent=@n5, BoneName=upperarm_l, BaseName=clavicle_l, DeltaComponentType=EulerX, InputRange=(X=37,Y=8), TwistAxis="1,0,0")
  ]}
]
variables [
 Assault Rifle {type=boolean}
 Assault rifle corr {type=boolean}
 Correction {type=number}
 Dual Wield corr {type=boolean}
 Frontend corr {type=boolean}
 "L Arm Pivot Euler Y 0 To -16 " {type=number}
 l Arm Raise 45 to 95 {type=number}
 L Arm Raise 60 To 105 {type=number}
 L Arm Raise 60 To 105 crouch Pistol {type=number}
 Left Pad crouch melee attacking alpha {type=number}
 Left Pad crouch targeting up alpha {type=number}
 Left pad emote corr {type=number}
 left shoulder pad arm twist corr {type=number}
 left shoulder pad up corr {type=number}
 Machine gun corr {type=boolean}
 Melee corr {type=boolean}
 Melee corr idle {type=boolean}
 Melee Spine rotation to the Left Alpha {type=number}
 Melee Spine rotation to the Right Alpha {type=number}
 Minigun corr {type=boolean}
 NewVar {type=number}
 NewVar_0 {type=number}
 Not in Front End {type=boolean}
 Pistol {type=boolean}
 R Arm Raise 45 to 95 {type=number}
 R Arm Raise 95 to 135 {type=number}
 R Arm Raise While Targeting {type=number}
 Right Pad alpha {type=number}
 Right pad emote corr {type=number}
 Right pad machine gun down corr {type=number}
 Right pad machine gun side corr {type=number}
 right shoulder pad arm twist corr {type=number}
 right shoulder pad up corr {type=number}
 Rocket launcher corr {type=boolean}
 Skeletal Mesh {type=SkeletalMeshComponent}
 Spine rotation to the Left Alpha {type=number}
 Spine rotation to the Right Alpha {type=number}
]
